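# Level Sequence for the haa1000 cinematic of PistolCanary, named as an animation sequence, running at a 60 fps display rate.
# It possesses a single CinematicNPC Superhero1 actor and has no sequence-level tracks.
# The actor's skeletal animation track has no sections, so it plays no animation.
# Its visibility track holds a constant false, keeping the NPC hidden while that section evaluates.
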LevelSequence /PistolCanary/Sequence/haa/haa1000/PistolCanary_haa1000_anm
bindings [PistolCanary_CinematicNPC_Superhero1 {
  id=#1
  possessed_class=PistolCanary_CinematicNPC_Superhero1_C
  tracks [
   {type=MovieSceneSkeletalAnimationTrack, sections []}
   {
    type=MovieSceneVisibilityTrack
    sections [{type=MovieSceneVisibilitySection, channels [None {type=MovieSceneScriptingBoolChannel, default=false, const [false,2]}]}]
   }
  ]
 }]
display_rate {denominator=1, numerator=60}
tick_resolution {denominator=1, numerator=24000}
tracks []
