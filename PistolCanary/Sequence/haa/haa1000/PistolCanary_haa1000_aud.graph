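# PistolCanary_haa1000_aud is a Level Sequence, apparently the audio layer of the haa1000 sequence, that drives the PistolCanary audio manager.
# It binds a possessable PistolCanary_AudioManager with no other bindings or root tracks.
# An event trigger track fires two events, whose target functions are not shown.
# A Hero Active Mix bool track switches on just after the first event and off on the same frame as the second.
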LevelSequence /PistolCanary/Sequence/haa/haa1000/PistolCanary_haa1000_aud
bindings [PistolCanary_AudioManager {
  id=#1
  possessed_class=PistolCanary_AudioManager_C
  tracks [
   {
    type=MovieSceneEventTrack
    sections [{
      type=MovieSceneEventTriggerSection
      channels [None {
        type=MovieSceneScriptingEventChannel
        keys [[3600, "<Struct 'MovieSceneEvent' (0x852f46aa0a0) {}>"], [4051, "<Struct 'MovieSceneEvent' (0x852f46aa460) {}>"]]
       }]
     }]
   }
   {
    type=MovieSceneBoolTrack
    display_name=Hero Active Mix
    sections [{
      type=MovieSceneBoolSection
      channels [None {type=MovieSceneScriptingBoolChannel, default=false, keys [[3473,false],[3606,true],[4051,false]]}]
     }]
   }
  ]
 }]
display_rate {denominator=1, numerator=60}
tick_resolution {denominator=1, numerator=24000}
tracks []
