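# VFX level sequence for PistolCanary cinematic haa1000 that stages two successive PolarityWinn sound-trail and impact effects, left then right.
# Each trail's Niagara actor travels along a keyed transform path toward the impact area while its system section runs.
# As each trail finishes moving, Auto Activate switches on for its impact system, with a constant explosion-scale parameter override.
# Both impacts share one fixed transform.
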LevelSequence /PistolCanary/Sequence/haa/haa1000/PistolCanary_haa1000_vfx
bindings [
 NiagaraComponent0 {
  id=#1
  parent_id=#2
  possessed_class=NiagaraComponent
  tracks [
   {type=MovieSceneNiagaraSystemTrack, sections [{type=MovieSceneNiagaraSystemSpawnSection, end_frame=3767, start_frame=3676}]}
   {
    type=MovieSceneBoolTrack
    display_name=Auto Activate
    sections [{type=MovieSceneBoolSection, channels [None {type=MovieSceneScriptingBoolChannel, default=true, const [true,1]}]}]
   }
  ]
 }
 NS_PC_PolarityWinn_SoundTrail_L {
  id=#2
  possessed_class=NiagaraActor
  tracks [{
    type=MovieScene3DTransformTrack
    sections [{
      type=MovieScene3DTransformSection
      channels [
       Location.X {type=MovieSceneScriptingDoubleChannel, default=6398, keys [[3677,6398],[3717,14280]]}
       Location.Y {type=MovieSceneScriptingDoubleChannel, default=70000, keys [[3677,78100],[3717,70000]]}
       Location.Z {type=MovieSceneScriptingDoubleChannel, default=7508, keys [[3677,3708],[3717,7508]]}
       Rotation.X {type=MovieSceneScriptingDoubleChannel, default=0}
       Rotation.Y {type=MovieSceneScriptingDoubleChannel, default=0}
       Rotation.Z {type=MovieSceneScriptingDoubleChannel, default=118.3055}
       Scale.X {type=MovieSceneScriptingDoubleChannel, default=1}
       Scale.Y {type=MovieSceneScriptingDoubleChannel, default=1}
       Scale.Z {type=MovieSceneScriptingDoubleChannel, default=1}
      ]
     }]
   }]
 }
 NS_PC_PolarityWinn_Impact_L {
  id=#3
  possessed_class=NiagaraActor
  tracks [{
    type=MovieScene3DTransformTrack
    sections [{
      type=MovieScene3DTransformSection
      channels [
       Location.X {type=MovieSceneScriptingDoubleChannel, default=13280}
       Location.Y {type=MovieSceneScriptingDoubleChannel, default=71104}
       Location.Z {type=MovieSceneScriptingDoubleChannel, default=7188}
       Rotation.X {type=MovieSceneScriptingDoubleChannel, default=-10.9397}
       Rotation.Y {type=MovieSceneScriptingDoubleChannel, default=-4.1587}
       Rotation.Z {type=MovieSceneScriptingDoubleChannel, default=-212.7323}
       Scale.X {type=MovieSceneScriptingDoubleChannel, default=1}
       Scale.Y {type=MovieSceneScriptingDoubleChannel, default=1}
       Scale.Z {type=MovieSceneScriptingDoubleChannel, default=1}
      ]
     }]
   }]
 }
 NiagaraComponent0 {
  id=#4
  parent_id=#3
  possessed_class=NiagaraComponent
  tracks [
   {type=MovieSceneNiagaraSystemTrack, sections [{type=MovieSceneNiagaraSystemSpawnSection, end_frame=3836, start_frame=3721}]}
   {
    type=MovieSceneNiagaraFloatParameterTrack
    display_name=System_ExplosionScale
    sections [{type=MovieSceneFloatSection, channels [None {type=MovieSceneScriptingFloatChannel, default=70}]}]
   }
   {
    type=MovieSceneBoolTrack
    display_name=Auto Activate
    sections [{type=MovieSceneBoolSection, channels [None {type=MovieSceneScriptingBoolChannel, default=false, keys [[3676,false],[3719,true]]}]}]
   }
  ]
 }
 NiagaraComponent0 {
  id=#5
  parent_id=#6
  possessed_class=NiagaraComponent
  tracks [
   {type=MovieSceneNiagaraSystemTrack, sections [{type=MovieSceneNiagaraSystemSpawnSection, end_frame=3815, start_frame=3741}]}
   {
    type=MovieSceneBoolTrack
    display_name=Auto Activate
    sections [{type=MovieSceneBoolSection, channels [None {type=MovieSceneScriptingBoolChannel, default=true, const [true,1]}]}]
   }
  ]
 }
 NiagaraComponent0 {
  id=#7
  parent_id=#8
  possessed_class=NiagaraComponent
  tracks [
   {type=MovieSceneNiagaraSystemTrack, sections [{type=MovieSceneNiagaraSystemSpawnSection, end_frame=3898, start_frame=3783}]}
   {
    type=MovieSceneNiagaraFloatParameterTrack
    display_name=System_ExplosionScale
    sections [{type=MovieSceneFloatSection, channels [None {type=MovieSceneScriptingFloatChannel, default=70}]}]
   }
   {
    type=MovieSceneBoolTrack
    display_name=Auto Activate
    sections [{type=MovieSceneBoolSection, channels [None {type=MovieSceneScriptingBoolChannel, default=false, keys [[3676,false],[3781,true]]}]}]
   }
  ]
 }
 NS_PC_PolarityWinn_Impact_R {
  id=#8
  possessed_class=NiagaraActor
  tracks [{
    type=MovieScene3DTransformTrack
    sections [{
      type=MovieScene3DTransformSection
      channels [
       Location.X {type=MovieSceneScriptingDoubleChannel, default=13280}
       Location.Y {type=MovieSceneScriptingDoubleChannel, default=71104}
       Location.Z {type=MovieSceneScriptingDoubleChannel, default=7188}
       Rotation.X {type=MovieSceneScriptingDoubleChannel, default=-10.9397}
       Rotation.Y {type=MovieSceneScriptingDoubleChannel, default=-4.1587}
       Rotation.Z {type=MovieSceneScriptingDoubleChannel, default=-212.7323}
       Scale.X {type=MovieSceneScriptingDoubleChannel, default=1}
       Scale.Y {type=MovieSceneScriptingDoubleChannel, default=1}
       Scale.Z {type=MovieSceneScriptingDoubleChannel, default=1}
      ]
     }]
   }]
 }
 NS_PC_PolarityWinn_SoundTrail_R {
  id=#6
  possessed_class=NiagaraActor
  tracks [{
    type=MovieScene3DTransformTrack
    sections [{
      type=MovieScene3DTransformSection
      channels [
       Location.X {type=MovieSceneScriptingDoubleChannel, default=14210, keys [[3741,6400],[3781,14210]]}
       Location.Y {type=MovieSceneScriptingDoubleChannel, default=70000, keys [[3741,78100],[3781,70000]]}
       Location.Z {type=MovieSceneScriptingDoubleChannel, default=7500, keys [[3741,3800],[3781,7500]]}
       Rotation.X {type=MovieSceneScriptingDoubleChannel, default=0}
       Rotation.Y {type=MovieSceneScriptingDoubleChannel, default=0}
       Rotation.Z {type=MovieSceneScriptingDoubleChannel, default=-40}
       Scale.X {type=MovieSceneScriptingDoubleChannel, default=1}
       Scale.Y {type=MovieSceneScriptingDoubleChannel, default=1}
       Scale.Z {type=MovieSceneScriptingDoubleChannel, default=1}
      ]
     }]
   }]
 }
]
display_rate {denominator=1, numerator=60}
tick_resolution {denominator=1, numerator=24000}
tracks []
